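; Blueprint for the Deimos Fiend, an AI enemy pawn in Athena, covering spawn, death, freeze and impact feedback.
; Authority hides it from the minimap and, on death, can grant the killer a shield GameplayEffect unless they are DBNO.
; Remote instances disable movement-component overlaps and create dynamic materials for time-stamped spawn and death erosion effects.
; GameplayCue events toggle a timed ice material and play surface-based impact sounds and particles.
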(event GameplayCue|GameplayCue.Athena.IceDeimos
  (params (EventType :type EGameplayCueEvent%20Enum :default "OnActive") (Parameters :type Gameplay%20Cue%20Parameters%20Structure%20%28by%20ref%29))
  ;[0 0.85 1] Application and removal of Chiller Ice material
  (switch Utilities|FlowControl|Switch|SwitchonEGameplayCueEvent EventType
    (:OnActive
      (Rendering|Material|SetScalarParameterValueonMaterials (Variables|Character|GetMesh) "Ice" 1.0)
      (Utilities|Time|SetTimerbyEvent (delegate Custom|RemoveChillerMaterial) (ScalableFloat|GetScalableFloatMultiplier (Variables|Default|GetFreezeDuration))))))

(event Custom|RemoveChillerMaterial
  ;[0 0.85 1] Application and removal of Chiller Ice material
  (Rendering|Material|SetScalarParameterValueonMaterials (Variables|Character|GetMesh) "Ice" 0.0))

(event EventBeginPlay
  ;[1 1 1] EVENT -> BeginPlay
  (|CallParentFunction)
  (Networking|SwitchHasAuthority
    (:Authority
      (EventDispatchers|BindEventtoOnDied (EventDispatchers|CreateEvent :!event Custom|CustomOnDied))
      (GameplayEffects|ApplyGameplayEffectToTarget :self (Variables|GameplayAbilities|GetAIPawnAbilitySystemComponent) :GameplayEffectClass "/Game/Athena/Deimos/Abilities/NPC/Generic/GE_HideOnMiniMap.GE_HideOnMiniMap_C" :Target (Variables|GameplayAbilities|GetAIPawnAbilitySystemComponent) :Level 0.0))
    (:Remote
      (bind _asprimitive_component (Utilities|Casting|CastToPrimitiveComponent (Variables|/Script/Engine.MovementComponent|GetUpdatedComponent (Variables|Character|GetCharacterMovement))))
      (Variables|/Script/Engine.PrimitiveComponent|SetbGenerateOverlapEvents false _asprimitive_component)
      (CallFunction|SetupMIDs))))

(event Spawning|EventOnStartedEncounterSpawn
  ;[1 1 1] Need Moved to Code to spawn GC
  (|CallParentFunction)
  (GameplayCue|ExecuteGameplayCue%28Burst%29Local self (Variables|Default|GetSpawnFXGC) (Variables|Default|GetCueParamsSpawnFiend))
  (Rendering|Material|SetScalarParameterValue (Variables|Default|GetMID_MonsterMaterial) "Spawn_CurrentTime" (Utilities|Time|GetGameTimeinSeconds) :!class /Script/Engine.MaterialInstanceDynamic))

(event Custom|CustomOnDied
  (params (DamagedActor :type Actor%20Object%20Reference) (Damage :type Float%20%28single-precision%29) (InstigatedBy :type Controller%20Object%20Reference) (DamageCauser :type Actor%20Object%20Reference) (HitLocation :type Vector) (FHitComponent :type Primitive%20Component%20Object%20Reference) (BoneName :type Name) (Momentum :type Vector))
  ;[1 1 1] Death material erosion
  (CallFunction|AnimateDeathMaterial)
  ;;[0.3509 0.6383 1] FORT-684077 | Add additional check to determine if rewarding shield in FNE modes
  (Networking|SwitchHasAuthority
    (:Authority
      (if (Fortnite|IsSpatialGameplayExperience)
        (if (GameplayTags|HasMatchingGameplayTag self "(TagName=\"Pawn.Athena.NPC.Deimos.DisableShieldOnElimination\")")
          _
          (else
            _))
        (else
          _))
      (bind _asfort_player_controller_athena (Utilities|Casting|CastToFortPlayerControllerAthena InstigatedBy))
      (Utilities|IsValid (Ability|GetAbilitySystemComponent (Game|Player|GetPlayerPawn _asfort_player_controller_athena))
        (:Is%20Valid
          (if (not (Health|IsDBNO (Game|Player|GetPlayerPawn _asfort_player_controller_athena)))
            (GameplayEffects|ApplyGameplayEffectToSelf :self (Ability|GetAbilitySystemComponent (Game|Player|GetPlayerPawn _asfort_player_controller_athena)) :GameplayEffectClass (Variables|Default|GetGE_DeimosShieldOnFiendKill) :Level 0.0)))))))

(event Custom|AnimateDeathMaterial
  (bind _mid_maskeddeathmaterial (Variables|Default|GetMID_MaskedDeathMaterial))
  ;[1 1 1] Death material erosion
  (Networking|SwitchHasAuthority
    (:Remote
      (Rendering|Material|SetMaterial (Variables|Character|GetMesh) 0 _mid_maskeddeathmaterial)
      (Rendering|Material|SetScalarParameterValue _mid_maskeddeathmaterial "Death_CurrentTime" (Utilities|Time|GetGameTimeinSeconds) :!class /Script/Engine.MaterialInstanceDynamic))))

(event GameplayCue|GameplayCue.Impact.Physical.Creature
  (params (EventType :type EGameplayCueEvent%20Enum :default "OnActive") (Parameters :type Gameplay%20Cue%20Parameters%20Structure%20%28by%20ref%29))
  (bind (_bblockinghit _binitialoverlap _time _distance _location) (Collision|BreakHitResult (Ability|GameplayCue|GetHitResult Parameters)))
  ;[1 1 1] ImpactATarget
  (Audio|PlaySoundatLocation (Sound|GetImpactPhysicalSurfaceSound (Ability|GameplayCue|GetHitResult Parameters)) _location)
  (Default|PlayDamageFXSurfaceEnemies Parameters (Variables|Effects|GetImpactParticles_Default) (Variables|Effects|GetImpactParticles_Wood) (Variables|Effects|GetImpactParticles_Stone) (Variables|Effects|GetImpactParticles_Metal) (Variables|Effects|GetImpactParticles_HumanFlesh) (.location (Transformation|GetSocketTransform (Variables|Character|GetMesh) (Variables|Default|GetImpact_MeleeSocket) "RTS_World")) (.rotation (Transformation|GetSocketTransform (Variables|Character|GetMesh) (Variables|Default|GetImpact_MeleeSocket) "RTS_World")) (Variables|Effects|GetMeleeImpact_UseSocketTransforms)))

(fn GetEnemyColor
  (return (Variables|Default|GetEnemyColor)))

(fn OnRep_HasBackPack)

(fn SetupMIDs
  (bind _mesh (Variables|Character|GetMesh))
  (bind _returnvalue (Rendering|Material|CreateDynamicMaterialInstance (Variables|Default|GetMaskedDeathMaterial) "None" "None" :!class /Script/Engine.KismetMaterialLibrary))
  (Variables|Default|SetMID_MaskedDeathMaterial _returnvalue)
  (bind _returnvalue_2 (Rendering|Material|CreateDynamicMaterialInstance (Rendering|Material|GetMaterial _mesh 0) "None" "None" :!class /Script/Engine.KismetMaterialLibrary))
  (Variables|Default|SetMID_MonsterMaterial _returnvalue_2)
  (Rendering|Material|SetMaterial _mesh 0 _returnvalue_2))

(fn Temp_FindClosestBone
  (params (HitLocation :type Vector))
  (bind _mesh (Variables|Character|GetMesh))
  (LocalVariables|SetLocalClosestBone "None")
  (LocalVariables|SetBoneDistance 0.0)
  (for _array_element (Transformation|GetAllSocketNames _mesh)
    (if (not (Utilities|String|Contains (Utilities|String|ToString(Name) _array_element) "FX"))
      (if (not (< (Math|Vector|VectorLength (- (Transformation|GetSocketLocation _mesh _array_element) HitLocation)) (LocalVariables|GetBoneDistance)))
        (Utilities|FlowControl|DoOnce false
          (:Completed
            _)))
      (LocalVariables|SetBoneDistance (Math|Vector|VectorLength (- (Transformation|GetSocketLocation _mesh _array_element) HitLocation)))
      (LocalVariables|SetLocalClosestBone _array_element)))
  (return (LocalVariables|GetLocalClosestBone)))

(fn ConstructionScript
  (|CallParentFunction))
